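# Animation Blueprint for the Sports Fashion Female pickaxe that continuously spins its blade bone, faster during melee animations.
# Each update sets roll speed by melee state (zero interp speed, so it snaps) and subtracts speed times delta time from the roll, applied additively to the blade.
# On initialization, the mesh component's name flags an offhand weapon, which enables a 180-degree yaw of b_AR_Root.
# An offhand branch for reversing spin direction has no incoming execution.
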AnimBlueprint /BRCosmetics/Weapons/FORT_Melee/Pickaxe_Sports_Fashion_Female/Meshes/Sports_Fashion_Female_Axe_AnimBP
target_skeleton=&/Sports_Fashion_Female_Axe_Skeleton
graphs [
 AnimGraph {nodes [
   n1 = "Transform (Modify) Bone - Bone: b_AR_Root"!(
    ComponentPose=Local To Component(LocalPose=Input Pose())
    bAlphaBoolEnabled:bool=Get offhand_weapon() -> (offhand_weapon:bool)
    Translation="0,0,0"
    Rotation="0,180,0"
    Scale="1,1,1"
   ) -> (Pose) {
    alphaInputType=Bool
    rotationMode=BMM_Additive
    rotationSpace=BCS_BoneSpace
    translationMode=BMM_Additive
    translationSpace=BCS_BoneSpace
    alphaBoolBlend {blendOption=Linear}
    boneToModify {boneName=b_AR_Root}
    rotation {yaw=180}
   }
   n2 = "Transform (Modify) Bone - Bone: blade"(ComponentPose=@n1, Alpha=1, Translation="0,0,0", Rotation=Get roll() -> (roll:Rotator), Scale="1,1,1") {
    rotationMode=BMM_Additive
    rotationSpace=BCS_BoneSpace
    translationMode=BMM_Additive
    translationSpace=BCS_BoneSpace
    alphaBoolBlend {blendOption=Linear}
    boneToModify {boneName=blade}
   }
   Output Pose(Result=Component To Local(ComponentPose=@n2)) {layerGroup=None}
  ]}
 EventGraph {nodes [
   n1 = Event BlueprintUpdateAnimation()
   > Set delta_seconds(delta_seconds:double=@n1.DeltaTimeX) -> (Output_Get:double)
   > Sequence()
   n2 = Get bIsPlayingMeleeAnim()
   n3 = Get delta_seconds() -> (delta_seconds:double)
   > n4 = Set roll_speed(roll_speed:double=FInterpTo(Current=Get roll_speed() -> (roll_speed:double), Target=SelectFloat(A=5000, B=2500, bPickA=@n2), DeltaTime=@n3, InterpSpeed=0)) -> (Output_Get:double)
   n5 = BreakRotator(InRot=Get roll())
   n6 = float * float(A=Get roll_speed(), B=@n3)
   Set!(
    execute=[@n9.then, @n4.then]
    roll_Roll:float=float - float(A=@n5, B=@n6) {is=title K2Node_CallFunction}
    roll_Pitch:float=0
    roll_Yaw:float=0
   ) -> (then, Output_Get:Rotator)
   Event BlueprintInitializeAnimation()
   > n7 = Set Skel_Mesh(Skel_Mesh=GetPartSkeletalMeshComponent())
   > n8 = Branch(Condition="Equal Exactly (String)"(A="To String (Object)"!(InObj:Object Reference=@n7) -> (ReturnValue:String), B=WeaponMeshOffhand))
   > Set offhand_weapon(offhand_weapon:bool=true) -> (Output_Get:bool)
   Set offhand_weapon(execute=@n8.else, offhand_weapon=false)
   n9 = Branch(Condition=Get offhand_weapon())
   Set(execute=@n9.else, roll_Roll=float + float(A=@n5, B=@n6) {is=title K2Node_CommutativeAssociativeBinaryOperator}, roll_Pitch=0, roll_Yaw=0)
   TryGetPawnOwner()
  ]}
]
variables [
 delta_seconds {type=number}
 offhand_weapon {type=boolean}
 roll {type=Rotator}
 roll_speed {type=number, default=5000}
 Skel_Mesh {type=SkeletalMeshComponent}
]
